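; Athena event Blueprint that shows sky cracks growing as a calendar event counts down, plus a looping crack effect after the launch sequence.
; After match start, the dedicated server watches the calendar event and maps days remaining to CrackProgression.
; Rep-notifies drive central crack visibility, three threshold-staggered small cracks with curve-driven material size, particles and progression-driven audio.
; Test events and construction script preview support cracks; a helper sets landscape river tint.
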(event Custom|TestCracks
  (params (DaysRemaining :type Float%20%28double-precision%29))
  ;[1 1 1] Test crack progression by days remaining
  (Variables|Default|SetCrackProgression (Math|Float|MapRangeClamped DaysRemaining (Variables|Default|GetCrackEventDuration) 0.0 0.0 3.0))
  (Variables|Default|SetAreCracksVisible 1))

(event Custom|TestCracksAnimated
  ;[1 1 1] Test crack spread animation
  (Variables|Default|SetAreCracksVisible 1)
  (bind (_direction _newtrack_0) (Timeline|Timeline PlayFromStart)
    (:Update
      (Variables|Default|SetCrackProgression _newtrack_0))))

(event Custom|ContinueSetup
  (if (not (Variables|Default|GetIgnoreCalendarEvent))
    (Utilities|FlowControl|Delay 1.0)
    (bind (_eventname _timeuntilend) (MCPContext|WatchCalendarEvent (Variables|Default|GetCrackEvent))
      (:OnEventActive
        (switch int (Variables|Default|GetAreEndSequenceCrackVisible)
          (:0
            (Variables|Default|SetCrackProgression (Math|Float|MapRangeClamped (- (Math|Timespan|GetTotalDays _timeuntilend) 1.0) (Variables|Default|GetCrackEventDuration) 0.0 1.0 3.0))
            (Variables|Default|SetAreCracksVisible 1)))))))

(event EventBeginPlay
  (Networking|CanExecuteCosmeticEvents
    (:False
      (bind _asfort_game_mode_zone (Utilities|Casting|CastToFortGameModeZone (Game|GetGameMode)))
      (EventDispatchers|BindEventtoOnHandleMatchHasStarted _asfort_game_mode_zone (EventDispatchers|CreateEvent :!event Custom|ContinueSetup))
      ;[1 1 1] Ensure rep notify triggers when changed from default
      (Variables|Default|SetAreEndSequenceCrackVisible 0)
      (Variables|Default|SetAreCracksVisible 0))))

(event Custom|EnablePostSequenceCrack
  ;[1 1 1] Turn on looping crack visual after launch sequence has finished
  (Variables|Default|SetAreEndSequenceCrackVisible 1)
  (Variables|Default|SetAreCracksVisible 1))

(fn OnRep_AreCracksVisible
  (switch int (Variables|Default|GetAreCracksVisible)
    (:1
      (Rendering|SetVisibility (Variables|Default|GetCentralCrack) true true)
      (Components|Activation|SetActive (Variables|Default|GetP_Geode_Crack_Center_Loop) true false)
      (if (not (Audio|Components|Audio|IsPlaying (Variables|Default|GetCrackAudioLoop)))
        (Audio|Components|Audio|Play (Variables|Default|GetCrackAudioLoop) 0.0)))))

(fn OnRep_AreEndSequenceCrackVisible
  (Components|Activation|SetActive (Variables|Default|GetP_Geode_Crack_Center_Loop) (Utilities|Select :Index (Variables|Default|GetAreEndSequenceCrackVisible) :Option%200 false :Option%201 true) false))

(fn OnRep_CrackProgression
  (for (_array_element _array_index) (Utilities|Array|MakeArray (Variables|Default|GetSmallCrack1) (Variables|Default|GetSmallCrack2) (Variables|Default|GetSmallCrack3))
    (CallFunction|UpdateCrackProgression _array_element (Variables|Default|GetCrackProgression) (Math|Conversions|ToFloat(Integer) _array_index)))
  (Audio|Parameter|SetFloatParameter (Variables|Default|GetCrackAudioLoop) "CrackProgression" (Math|Float|Clamp(Float) (Variables|Default|GetCrackProgression) 0.0 3.0) :!class /Script/Engine.AudioComponent))

(fn OnRep_CracksVisible
  (if (Variables|Default|GetCracksVisible)
    (Rendering|SetVisibility (Variables|Default|GetCentralCrack) true true)
    (Components|Activation|SetActive (Variables|Default|GetP_Geode_Crack_Center_Loop) true false)
    (if (not (Audio|Components|Audio|IsPlaying (Variables|Default|GetCrackAudioLoop)))
      (Audio|Components|Audio|Play (Variables|Default|GetCrackAudioLoop) 0.0))))

(fn OnRep_EndSequenceCrackVisible
  (Components|Activation|SetActive (Variables|Default|GetP_Geode_Crack_Center_Loop) (Variables|Default|GetEndSequenceCrackVisible) false))

(fn SetupDynamicMaterials
  (params (Mesh :type Static%20Mesh%20Component%20Object%20Reference) (MatArray :type Array%20of%20Material%20Instance%20Dynamic%20Object%20References))
  (for _index (range (Utilities|Array|Length (Rendering|Material|GetMaterials Mesh)))
    (bind _returnvalue (Rendering|Material|CreateDynamicMaterialInstance Mesh _index 0 "None" :!class /Script/Engine.PrimitiveComponent))
    (Utilities|Array|Add MatArray _returnvalue)))

(fn SetupRiverTint
  (params (EnableTint :type Boolean) (TintColor :type Linear%20Color%20Structure))
  (if EnableTint
    (Rendering|Material|SetVectorParameterValue "/Game/Athena/Environments/Landscape/MPC/MPC_Landscape.MPC_Landscape" "RiverTintAdd" TintColor :!class /Script/Engine.KismetMaterialLibrary)
    (else
      (Rendering|Material|SetVectorParameterValue "/Game/Athena/Environments/Landscape/MPC/MPC_Landscape.MPC_Landscape" "RiverTintAdd" (Utilities|Struct|MakeLinearColor 0.0) :!class /Script/Engine.KismetMaterialLibrary))))

(fn UpdateCrackProgression
  (params (Mesh :type Mesh%20Component%20Object%20Reference) (Progression :type Float%20%28double-precision%29) (Threshold :type Float%20%28double-precision%29))
  (Rendering|SetVisibility Mesh (> Progression Threshold) true)
  (Rendering|Material|SetScalarParameterValueonMaterials Mesh "CrackSize" (Math|Curves|GetFloatValue (Variables|Default|GetCrackGrowthCurve) (Math|Float|Clamp(Float) (- Progression Threshold))))
  (if (>= (Variables|Default|GetCrackProgression) 2.8)
    (Components|Activation|Activate (Variables|Default|GetP_Crack_Sand) false)))

(fn ConstructionScript
  (Utilities|FlowControl|Sequence
    (:then_1
      ;[1 1 1] Update sky cracks
      (if (Variables|Default|GetPreviewCracks)
        (for (_array_element _array_index) (Utilities|Array|MakeArray (Variables|Default|GetSmallCrack1) (Variables|Default|GetSmallCrack2) (Variables|Default|GetSmallCrack3))
          (CallFunction|UpdateCrackProgression _array_element (Variables|Default|GetCrackProgression) (Math|Conversions|ToFloat(Integer) _array_index)))
        (Rendering|SetVisibility (Variables|Default|GetCentralCrack) true false)
        (Components|Activation|SetAutoActivate (Variables|Default|GetP_Geode_Crack_Center_Loop) true)))))
